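# Prototype translucent unlit material for Niagara mesh particles in the AnnoyedMop VFX, producing a scrolling, eroding, wobbling energy effect.
# Two panning fire textures form a mask that, inverted, tints Color; a water texture adds Highlights before Tint and emissive adjustment.
# That mask, vertex red and a particle-alpha-driven smoothstep erosion form opacity.
# Dynamic parameters offset time and segment UVs, which noise distorts; panned noise also builds a normal-direction vertex wobble.
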Material /AnnoyedMop/Assets/VFX/proto/M_FX_AnnoyedMopProto_Headline
graph {
 properties {
  bCastRayTracedShadows=false
  bUsedWithNiagaraMeshParticles=true
  blendMode=BLEND_Translucent
  pixelDepthOffsetMode=PDOM_Legacy
  shadingModel=MSM_Unlit
  usedShadingModels=Unlit
 }
 nodes [
  n1 = TextureCoordinate() {uTiling=1.25, vTiling=0.5}
  n2 = Add(A=Time(), B=DynamicParameter() {parameterIndex=2}.G)
  n3 = Multiply(A=TextureSample(UVs=Panner(Coordinate=@n1, Time=@n2) {speedY=1}) {texture=&T_NoisePack_01}.R) {constB=2}
  n4 = VertexNormalWS()
  n5 = ScalarParameter() {group=Wobble, parameterName=VertexWobble}
  Add(A=Multiply(A=Add(A=@n3) {constB=-1}, B=Multiply(A=@n4, B=@n5)))
  n6 = VectorParameter() {parameterName=TilingC, defaultValue=<rgba 1.5 1 0 1>}
  n7 = Multiply(A=NamedRerouteUsage(), B=ComponentMask(None=@n6.RGB) {g=true, r=true})
  n8 = NamedRerouteUsage()
  n9 = AppendVector(A=@n6.B, B=@n6.B)
  n10 = Multiply(
   A=VectorParameter() {parameterName=Highlights, defaultValue=<rgba 0 0 0 1>}.RGB
   B=TextureSample(UVs=Panner(Coordinate=@n7, Time=@n8, Speed=@n9) {speedX=-0.5}) {texture=/Game/Effects/Fort_Effects/Textures/T_PlantStand_Water02}.G
  )
  n11 = NamedRerouteUsage()
  n12 = VectorParameter() {parameterName=TilingA, defaultValue=<rgba 1.5 1 0 1>}
  n13 = ComponentMask(None=@n12.RGB) {g=true, r=true}
  n14 = TextureSample(UVs=Panner(Coordinate=Multiply(A=@n11, B=@n13), Time=@n8, Speed=AppendVector(A=@n12.B, B=@n12.B)) {speedX=1}) {texture=/Game/Effects/Fort_Effects/Textures/T_PlantStand_Fire}
  n15 = Add(A=Add(A=@n14.R, B=@n14.R), B=@n14.B)
  n16 = NamedRerouteUsage()
  n17 = VectorParameter() {parameterName=TilingB, defaultValue=<rgba 1.5 1 0 1>}
  n18 = ComponentMask(None=@n17.RGB) {g=true, r=true}
  n19 = TextureSample(UVs=Panner(Coordinate=Multiply(A=@n16, B=@n18), Time=@n8, Speed=AppendVector(A=@n17.B, B=@n17.B)) {speedX=-0.5}) {texture=&T_PlantStand_Fire}
  n20 = Add(A=Add(A=@n19.R, B=@n19.R), B=@n19.B)
  n21 = Divide(A=Add(A=Divide(A=@n15) {constB=3}, B=Divide(A=@n20) {constB=3}))
  n22 = Multiply(A=VectorParameter() {parameterName=Color, defaultValue=<rgba 0 0 0 1>}.RGB, B=OneMinus(None=@n21))
  MaterialFunctionCall(Emissive=Multiply(A=VectorParameter() {parameterName=Tint, defaultValue=<rgba 0 0 0 1>}.RGB, B=Add(A=@n10, B=@n22))) {materialFunction=&MF_FX_EmissiveAdjustments}
  n23 = OneMinus(None=ParticleColor().A)
  n24 = ScalarParameter() {parameterName=erode_w}
  n25 = Panner(Coordinate=Multiply(A=NamedRerouteUsage(), B=Constant2Vector() {g=0.2})) {bFractionalPart=true, speedY=-1}
  Multiply(
   A=Multiply(A=@n21, B=VertexColor().R)
   B=SmoothStep(
    Min=@n23
    Max=Add(A=@n23, B=@n24)
    Value=TextureSample(UVs=@n25) {samplerType=SAMPLERTYPE_Masks, texture=/AnnoyedMop/Assets/VFX/Textures/T_FX_AnnoyedMop_Prj}.R
   )
  )
  n26 = DynamicParameter() {parameterIndex=2, paramNames [Param1,Param2,Tiling,Thickness]}
  n27 = AppendVector(A=@n26.Tiling, B=Constant() {r=1})
  n28 = Add(A=AppendVector(A=@n26.Param2, B=Constant()), B=Multiply(A=TextureCoordinate() {uTiling=0.25}, B=@n27))
  n29 = ScalarParameter() {defaultValue=0.1, parameterName=Distortion}
  n30 = Multiply(A=TextureCoordinate() {uTiling=0.25, vTiling=0.5}, B=ScalarParameter() {defaultValue=1, parameterName=NoiseTile})
  n31 = Add(A=@n26.Param2, B=Time())
  n32 = TextureSample(UVs=Panner(Coordinate=Multiply(A=@n27, B=@n30), Time=@n31) {speedY=-2}) {samplerType=SAMPLERTYPE_Masks, texture=&T_FX_AnnoyedMop_Prj}
  n33 = Constant2Vector() {g=1}
  n34 = Subtract(A=Multiply(A=Multiply(A=@n32.R, B=@n33)) {constB=2})
  NamedRerouteDeclaration(None=Add(A=NamedRerouteDeclaration(None=@n28) {name=UnscaledSegmentUV, nodeColor=<rgba 1 0.6328 0 1>}, B=Multiply(A=@n29, B=@n34))) {name=SegmentUV, nodeColor=<rgba 1 0 0.2578 1>}
  NamedRerouteDeclaration(None=@n31) {name=T, nodeColor=<rgba 1 0 0.1172 1>}
  ScalarParameter() {parameterName=erode_p}
  TextureCoordinate()
  VectorParameter() {parameterName=Vertex_Wobble, defaultValue=<rgba 0 0 0 1>}
 ]
}
