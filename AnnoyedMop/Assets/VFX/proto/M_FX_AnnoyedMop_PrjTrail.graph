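# Translucent unlit prototype VFX material for the AnnoyedMop projectile trail, usable on Niagara ribbons or mesh particles via the IsRibbon switch.
# Two panning samples of the projectile texture add highlights to particle color through MF_FX_EmissiveAdjustments.
# Their combined masks, a trail-head texture, length falloff and smoothstep erosion shape opacity.
# Optional world position offset combines noise wobble along normals with dynamic-parameter displacement, weighted along UV V.
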Material /AnnoyedMop/Assets/VFX/proto/M_FX_AnnoyedMop_PrjTrail
graph {
 properties {
  bUsedWithNiagaraMeshParticles=true
  bUsedWithNiagaraRibbons=true
  blendMode=BLEND_Translucent
  pixelDepthOffsetMode=PDOM_Legacy
  shadingModel=MSM_Unlit
  usedShadingModels=Unlit
 }
 nodes [
  n1 = NamedRerouteUsage()
  n2 = Constant2Vector() {g=2, r=0.5}
  n3 = StaticSwitchParameter(True=Multiply(A=Time(), B=ScalarParameter() {defaultValue=1, parameterName=PanningSpeed}), False=Time()) {parameterName=IsRibbon}
  n4 = TextureSample(UVs=Panner(Coordinate=Multiply(A=@n1, B=@n2), Time=@n3) {speedY=-3}) {samplerType=SAMPLERTYPE_Masks, texture=/AnnoyedMop/Assets/VFX/Textures/T_FX_AnnoyedMop_Prj}
  n5 = Multiply(A=VectorParameter() {parameterName=Highlights, defaultValue=<rgba 0 0 0 1>}.RGB, B=@n4.G)
  n6 = NamedRerouteUsage()
  n7 = Constant2Vector() {g=3, r=0.75}
  n8 = TextureSample(UVs=Panner(Coordinate=Multiply(A=@n6, B=@n7), Time=@n3) {speedY=-4}) {samplerType=SAMPLERTYPE_Masks, texture=&T_FX_AnnoyedMop_Prj}
  n9 = TextureCoordinate()
  n10 = StaticSwitchParameter(True=MaterialFunctionCall(XY=@n9) {materialFunction=&Swizzle}.YX, False=@n9) {parameterName=IsRibbon}
  n11 = ComponentMask(None=@n10) {g=true}
  n12 = Multiply(
   A=Multiply(A=VectorParameter() {parameterName=HighlightsCore, defaultValue=<rgba 0 0 0 1>}.RGB, B=@n8.G)
   B=Power(Base=OneMinus(None=@n11)) {constExponent=8}
  )
  MaterialFunctionCall(Emissive=Add(A=ParticleColor().RGB, B=Add(A=@n5, B=@n12))) {materialFunction=&MF_FX_EmissiveAdjustments}
  n13 = Power(Base=OneMinus(None=ComponentMask(None=@n10) {g=true})) {constExponent=3}
  n14 = Multiply(
   A=Add(A=@n4.G, B=@n8.G)
   B=TextureSample(UVs=@n10) {samplerType=SAMPLERTYPE_Masks, texture=/AnnoyedMop/Assets/VFX/Textures/T_AnnoyedMop_TrailHead}.R
  )
  n15 = ParticleColor()
  n16 = StaticSwitchParameter(True=DynamicParameter() {paramNames [ErodeP,Param2,Param3,Param4]}.ErodeP, False=@n15.A) {parameterName=IsRibbon}
  n17 = Add(A=@n16, B=ScalarParameter() {parameterName=erode_w})
  n18 = Divide(A=Add(A=@n4.R, B=@n8.R))
  Multiply(
   A=Multiply(A=Multiply(A=@n13, B=@n14), B=SmoothStep(Min=@n16, Max=@n17, Value=@n18))
   B=StaticSwitchParameter(True=@n15.A, False=Constant() {r=0.35}) {parameterName=IsRibbon}
  ) {constB=0.35}
  Multiply()
  n19 = MaterialFunctionCall(XY=TextureCoordinate() {coordinateIndex=1}) {materialFunction=&Swizzle}
  n20 = DynamicParameter() {parameterIndex=2, paramNames [Param1,Offset,Time,Param4]}
  n21 = TextureCoordinate()
  NamedRerouteDeclaration(None=StaticSwitchParameter(True=Add(A=@n19.YX, B=@n20.Offset), False=Add(A=@n21, B=@n20.Offset)) {parameterName=IsRibbon}) {name=SegmentUV, nodeColor=<rgba 0.5234 1 0 1>}
  n22 = ComponentMask(None=TextureCoordinate()) {g=true}
  OneMinus(None=@n22)
  ScalarParameter() {parameterName=erode_p}
  n23 = DynamicParameter() {parameterIndex=1}
  n24 = AppendVector(A=@n23.Param1, B=@n23.Param1)
  n25 = Multiply(A=ComponentMask(None=TextureCoordinate()) {g=true}, B=Multiply(A=AppendVector(A=@n24, B=@n23.Param3), B=@n23.Param4))
  n26 = VertexNormalWS()
  n27 = Time()
  n28 = DynamicParameter() {parameterIndex=2, paramNames [Param1,Offset,Time,Param4]}
  n29 = TextureSample(UVs=Panner(Coordinate=TextureCoordinate() {uTiling=1.25, vTiling=0.5}, Time=Add(A=@n27, B=@n28.Offset)) {speedY=-1}) {texture=&T_NoisePack_01}
  n30 = Multiply(A=@n22, B=Multiply(A=Multiply(A=@n26, B=@n29.R), B=ScalarParameter() {parameterName=Wobble}))
  n31 = DynamicParameter() {paramNames [ErodeP,Param2,Param3,Param4]}
  n32 = Subtract(A=AppendVector(A=AppendVector(A=@n31.ErodeP, B=@n31.ErodeP), B=@n31.Param3), B=WorldPosition().XYZ)
  n33 = ComponentMask(None=TextureCoordinate()) {g=true}
  StaticSwitchParameter(True=LinearInterpolate(A=Add(A=@n25, B=@n30), B=Multiply(A=@n32, B=@n33), Alpha=@n33), False=Constant3Vector()) {defaultValue=true, parameterName=bUseWPO}
  StaticSwitchParameter(True=Constant() {r=1}) {parameterName=IsRibbon}
  TextureCoordinate()
 ]
}
