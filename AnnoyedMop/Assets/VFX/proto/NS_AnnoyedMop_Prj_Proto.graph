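# Prototype Niagara system for the AnnoyedMop projectile: a fiery, segmented projectile built from user arrays of segment positions, rotations and active flags.
# SegmentProjectile and SegmentedTrails burst one particle per segment for meshes and lights, joined by ribbon trails and GPU embers.
# Particles partition by segment and die where ActiveArray is false, so hit segments drop out.
# GPU smoke shoots downward and collides; both flame-lick emitters are disabled.
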NiagaraSystem /AnnoyedMop/Assets/VFX/proto/NS_AnnoyedMop_Prj_Proto
dependencies {
 usedDataInterfaces [&NiagaraDataInterfaceCurve, &NiagaraDataInterfaceColorCurve, &NiagaraDataInterfacePlatformSet]
 usedDynamicInputs [&self:GetArrayLength, &FN_PickFloatDependOnPlatform]
 usedModules [
  &self:SystemSpawnScript.SystemScriptSource.SystemScriptGraph.NiagaraNodeAssignment_2.SetVariables_#1
  &self:SystemSpawnScript.SystemScriptSource.SystemScriptGraph.NiagaraNodeAssignment_3.SetVariables_#2
  &self:SystemSpawnScript.SystemScriptSource.SystemScriptGraph.NiagaraNodeAssignment_4.SetVariables_#3
  &self:SystemSpawnScript.SystemScriptSource.SystemScriptGraph.NiagaraNodeAssignment_5.SetVariables_#4
  &self:SystemSpawnScript.SystemScriptSource.SystemScriptGraph.NiagaraNodeAssignment_6.SetVariables_#5
  &self:SystemSpawnScript.SystemScriptSource.SystemScriptGraph.NiagaraNodeAssignment_7.SetVariables_#6
  &self:SystemSpawnScript.SystemScriptSource.SystemScriptGraph.NiagaraNodeAssignment_10.SetVariables_#7
  &self:SystemSpawnScript.SystemScriptSource.SystemScriptGraph.NiagaraNodeAssignment_11.SetVariables_#8
  &self:SystemSpawnScript.SystemScriptSource.SystemScriptGraph.NiagaraNodeAssignment_9.SetVariables_#9
  &self:SystemSpawnScript.SystemScriptSource.SystemScriptGraph.NiagaraNodeAssignment_8.SetVariables_#10
  &self:SystemSpawnScript.SystemScriptSource.SystemScriptGraph.NiagaraNodeAssignment_13.SetVariables_#11
  &self:SystemSpawnScript.SystemScriptSource.SystemScriptGraph.NiagaraNodeAssignment_14.SetVariables_#12
  &SystemState
  &self:SystemSpawnScript.SystemScriptSource.SystemScriptGraph.NiagaraNodeAssignment_0.SetVariables_#13
  &self:SystemSpawnScript.SystemScriptSource.SystemScriptGraph.NiagaraNodeAssignment_1.SetVariables_#14
  &Timeline
  &self:SystemSpawnScript.SystemScriptSource.SystemScriptGraph.NiagaraNodeAssignment_12.SetVariables_#15
 ]
}
emitters {
 FlameLicks {
  bEnabled=false
  emitterUpdateScript [
   &EmitterState()
   &SpawnPerUnit(Module.Use Max Movement Threshold=true, Module.Use Movement Tolerance=false, Use Spawn Probability=false) {enabled=false}
   &SpawnRate(SpawnRate=0, Use Spawn Probability=false) {enabled=false, module=SpawnRate001}
   &SpawnRate(SpawnRate=50, Use Spawn Probability=false)
  ]
  particleSpawnScript [
   &InitializeParticle(
    Lifetime=&RandomRangeFloat(Minimum=0.25, Maximum=0.6)
    Sprite Size Mode=::NewEnumerator3
    Uniform Sprite Size=&RandomRangeFloat(Minimum=20, Maximum=40)
    UsePositionOffset=false
    Write Material Random=true
   )
   &PartitionParticles(Number Of Partitions=$User.NumSegments)
   &ShapeLocation(
    Additional Quaternion Rotation=false
    Box Size=<5 100 5>
    MaskSpawnGroup=false
    Offset=&Multiply_Vector(
     A=&SelectVectorFromArray(Direct Array Index=$Particles.PartitionParticles.Partition, Vector Selection Array=$User.SegmentPositions)
     B=$User.RootScale
    )
    Offset Mode=::NewEnumerator0
    Override Seed=false
    Rotation Mode=::NewEnumerator3
    Rotation Quaternion=&self:ScratchDynamicInput_02(Index:NiagaraInt32=$Particles.PartitionParticles.Partition, QuatArray:NiagaraDataInterfaceArrayQuat=$User.SegmentRotations)
    Shape Primitive=::NewEnumerator2
    Surface Only=false
    UseOwnerScale=true
   )
   &AddVelocity(
    Cone Angle=2
    Cone Axis=<-1 0 0>
    Module.UseConeFalloff=false
    Module.UseVelDistribution=false
    Override Seed=false
    Velocity Mode=::NewEnumerator2
    Velocity Speed=&RandomRangeFloat(Minimum=750, Maximum=1600)
   )
  ]
  particleUpdateScript [
   &ParticleState()
   &ScaleSpriteSize(Uniform Curve Sprite Scale=DataInterface() {propertyValues {Curve {keys [<at 0 0 RCIM_Cubic>, {arriveTangent=1.375, interpMode=RCIM_Cubic, leaveTangent=1.375, time=0.1, value=0.95}, <at 1 1.375 RCIM_Cubic>]}}})
   &KillParticles(Kill Particles=&InvertBool(Boolean:NiagaraBool=&self:ScratchDynamicInput(SegmentIndex:NiagaraInt32=$Particles.PartitionParticles.Partition, VisibilityArray:NiagaraDataInterfaceArrayBool=$User.ActiveArray)))
   &Drag(Drag=3, Use Linear Drag=true, Use Rotational Drag=true)
   &SolveForcesAndVelocity(Clamp Velocity=false, Limit Acceleration=false)
   &ScaleSpriteSizeBySpeed(Max Scale Factor=<2 8>)
   &ScaleColor(ScaleA=true, ScaleRGB=true)
   &SubUVAnimation(
    End Frame=15
    Start Frame:NiagaraInt32=&RandomRangeInt(Minimum=0, Maximum=5)
    SubUV Loop Count:NiagaraFloat=&RandomRangeFloat(Minimum=1, Maximum=1)
   )
   &Color(
    Color=&ColorFromCurve(ColorCurve=DataInterface() {propertyValues {
       AlphaCurve {keys [<at 0.1109 1 RCIM_Cubic>,<at 1 0 RCIM_Cubic>]}
       BlueCurve {keys [<at 0 0 RCIM_Cubic>,<at 0.1272 0>,<at 0.4089 0>]}
       GreenCurve {keys [<at 0 2.0242 RCIM_Cubic>,<at 0.1272 0.0329>,<at 0.4089 0.1707>]}
       RedCurve {keys [<at 0 5 RCIM_Cubic>,<at 0.1272 1>,<at 0.4089 1>]}
      }})
    Scale Alpha=0.25
   )
  ]
  rendererClasses [&NiagaraSpriteRendererProperties]
  renderers [{rendererClass=&NiagaraSpriteRendererProperties}]
 }
 GPU_Collided_Smoke {
  simTarget=GPUComputeSim
  emitterUpdateScript [
   &EmitterState(Life Cycle Mode=::NewEnumerator1, Loop Behavior=::NewEnumerator1, Loop Delay=0.05, Loop Duration=0.25, UseLoopDelay=true)
   &SpawnRate(SpawnRate=&Multiply_FloatByInt(Float=100, Integer=$User.NumSegments), Use Spawn Probability=false)
  ]
  particleSpawnScript [
   &InitializeParticle(
    Color=<rgba 1 1 1 0.4>
    Lifetime Max=0.35
    Lifetime Min=0.2
    Lifetime Mode=::NewEnumerator1
    Sprite Rotation Mode=::NewEnumerator2
    Sprite Size Mode=::NewEnumerator1
    Sprite UV Mode=::NewEnumerator3
    Uniform Sprite Size Max=250
    Uniform Sprite Size Min=200
    UsePositionOffset=false
    Write Material Random=true
   )
   &ShapeLocation(MaskSpawnGroup=false, Non Uniform Scale=<0.8 3 0>, Override Seed=false, UseOwnerScale=true)
   &AddVelocity(Rotation Coordinate Space=::World, Velocity=<0 0 -4000>)
   &CameraOffset(Camera Offset Amount=80)
  ]
  particleUpdateScript [
   &ParticleState()
   &ScaleSpriteSize(
    Uniform Curve Scale=&MakeCustomFloatFromBool(
     Bool=&SetBoolByFloatComparison(A=2, B=HlslExpression() {hlslExpression="abs(Particles.Velocity.x + Particles.Velocity.y + Particles.Velocity.z)"})
     True Float=1.5
     False Float=0
    )
    Uniform Curve Sprite Scale=DataInterface() {propertyValues {Curve {keys [
        {tangentMode=RCTM_User, time=0, value=0.4}
        {arriveTangent=-0.0405, interpMode=RCIM_Cubic, leaveTangent=-0.0405, tangentMode=RCTM_User, time=1, value=1}
       ]}}}
   )
   &Collision(
    Age Colliding Particles=false
    Collision Enabled=&SetBoolByFloatComparison(
     A=1
     B=&FloatFromCurve(FloatCurve=DataInterface() {propertyValues {Curve {keys [<at 0.18 0.95>,<at 1 2>]}}})
    )
    Enable Rest State=false
    Friction=5
    Friction During a Bounce=5
    GPU Collision Type=::NewEnumerator2
    Module.EnableMaxCollisionCount=false
    Module.Randomize Collision Normal Vector=true
    Particle Radius Scale=0
    Randomize Collision Normal=0
    Restitution=0
    Simple Friction=false
   )
   &SolveForcesAndVelocity(Clamp Velocity=false, Limit Acceleration=false)
   &DynamicMaterialParameters(
    Index 0 Param 1=&RandomRangeFloat(Minimum=0, Maximum=1)
    Module.Param0WriteEnabled=true
    Module.Param1WriteEnabled=false
    Module.Param2WriteEnabled=false
    Module.Param3WriteEnabled=false
   )
   &ScaleColor(
    Scale Alpha=&MakeCustomFloatFromBool(
     Bool=&SetBoolByFloatComparison(A=10, B=HlslExpression() {hlslExpression="abs(Particles.Velocity.x + Particles.Velocity.y + Particles.Velocity.z)"})
     True Float=&FloatFromCurve(
      FloatCurve=DataInterface() {propertyValues {Curve {keys [
          {tangentMode=RCTM_User, time=0, value=0}
          <at 0.1 0.975>
          {arriveTangent=-0.0405, interpMode=RCIM_Cubic, leaveTangent=-0.0405, tangentMode=RCTM_User, time=0.3, value=1}
          {arriveTangent=-0.3723, interpMode=RCIM_Cubic, leaveTangent=-0.3723, tangentMode=RCTM_User, time=1, value=0.05}
         ]}}}
      Scale Curve=&FNTimeOfDayScalar(Night Scale=1, Morning Scale=1, Evening Scale=1)
     )
     False Float=0
    )
    ScaleA=true
    ScaleRGB=true
   )
  ]
  rendererClasses [&NiagaraSpriteRendererProperties]
  renderers [{rendererClass=&NiagaraSpriteRendererProperties}]
 }
 LongFlamelicks {
  bEnabled=false
  emitterUpdateScript [
   &EmitterState(Life Cycle Mode=::NewEnumerator1, Loop Behavior=::NewEnumerator1, Loop Duration Mode=::NewEnumerator1, UseLoopDelay=false)
   &SpawnPerUnit(Module.Use Max Movement Threshold=false, Module.Use Movement Tolerance=false, Spawn Spacing=100, Use Spawn Probability=true)
   &SpawnRate(SpawnRate=0, Use Spawn Probability=false) {enabled=false, module=SpawnRate001}
   &SpawnRate(
    SpawnRate=&Multiply_Float(A=&Multiply_FloatByInt(Float=10, Integer=$User.NumSegments), B=$System.GlobalFade)
    Use Spawn Probability=false
   )
  ]
  particleSpawnScript [
   &InitializeParticle(
    Lifetime=&RandomRangeFloat(Minimum=0.12, Maximum=0.2)
    Sprite Size Mode=::NewEnumerator3
    Uniform Sprite Size=&RandomRangeFloat(Minimum=40, Maximum=55)
    UsePositionOffset=false
    Write Material Random=true
   )
   &PartitionParticles(Number Of Partitions=$User.NumSegments)
   &ShapeLocation(
    Additional Quaternion Rotation=false
    Box Size=<5 100 5>
    MaskSpawnGroup=false
    Offset=&Multiply_Vector(
     A=&SelectVectorFromArray(Direct Array Index=$Particles.PartitionParticles.Partition, Vector Selection Array=$User.SegmentPositions)
     B=$User.RootScale
    )
    Offset Mode=::NewEnumerator0
    Override Seed=false
    Rotation Mode=::NewEnumerator3
    Rotation Quaternion=&self:ScratchDynamicInput_02(Index:NiagaraInt32=$Particles.PartitionParticles.Partition, QuatArray:NiagaraDataInterfaceArrayQuat=$User.SegmentRotations)
    Shape Primitive=::NewEnumerator2
    Surface Only=false
    UseOwnerScale=true
   )
   &AddVelocity(
    Cone Angle=1
    Cone Axis=&NormalizeVector(
     Vector To Normalize=&Subtract_Vector(A=$System.TrailTargetPosition, B=Linked() {linkedVariable {name=Particles.Position, type=NiagaraPosition}})
     Safe Normalize=true
     Safe Fallback Threshold=0
     Safe Fallback Vector=<0 0 0.001>
    )
    Module.UseConeFalloff=false
    Module.UseVelDistribution=false
    Override Seed=false
    Velocity Mode=::NewEnumerator2
    Velocity Speed=&RandomRangeFloat(Minimum=750, Maximum=1600)
   )
   &KillParticles(Kill Particles=&InvertBool(Boolean:NiagaraBool=&self:ScratchDynamicInput(SegmentIndex:NiagaraInt32=$Particles.PartitionParticles.Partition, VisibilityArray:NiagaraDataInterfaceArrayBool=$User.ActiveArray)))
   /SetVariables(Particles.MaterialRandom:NiagaraFloat=&RandomRangeFloat(Minimum=0, Maximum=1)) {isSetParameters=true}
  ]
  particleUpdateScript [
   &ParticleState()
   &ScaleSpriteSize(Uniform Curve Sprite Scale=DataInterface() {propertyValues {Curve {keys [<at 0 0 RCIM_Cubic>, {arriveTangent=1.375, interpMode=RCIM_Cubic, leaveTangent=1.375, time=0.1, value=0.95}, <at 1 1.375 RCIM_Cubic>]}}})
   &Drag(Drag=3, Use Linear Drag=true, Use Rotational Drag=true)
   &SolveForcesAndVelocity(Clamp Velocity=false, Limit Acceleration=false)
   &ScaleSpriteSizeBySpeed(Max Scale Factor=<2 8>) {enabled=false}
   &ScaleColor(Scale Alpha=&Multiply_Float(A=$System.GlobalFade, B=1), ScaleA=true, ScaleRGB=true)
   &Color(
    Color=&ColorFromCurve(ColorCurve=DataInterface() {propertyValues {
       AlphaCurve {keys [<at 0.1109 1 RCIM_Cubic>,<at 1 0 RCIM_Cubic>]}
       BlueCurve {keys [<at 0 3.0621 RCIM_Cubic>,<at 0.1568 0>,<at 0.7467 0>]}
       GreenCurve {keys [<at 0 12.0761 RCIM_Cubic>,<at 0.1568 0.5655>,<at 0.7467 0.0274>]}
       RedCurve {keys [<at 0 20 RCIM_Cubic>,<at 0.1568 3>,<at 0.7467 0.4844>]}
      }})
    Scale Alpha=5
    Scale Color=<5 5 5>
   )
   &ScaleSpriteSize(
    Scale Factor=&MakeVector2D(
     X=&FloatFromCurve(
      FloatCurve=DataInterface() {propertyValues {Curve {keys [
          {arriveTangent=4.1785, interpMode=RCIM_Cubic, leaveTangent=4.1785, tangentMode=RCTM_User, time=0, value=0}
          {arriveTangent=1.1179, interpMode=RCIM_Cubic, leaveTangent=1.1179, tangentMode=RCTM_User, time=0.2042, value=0.6228}
          {arriveTangent=0.0224, interpMode=RCIM_Cubic, leaveTangent=0.0224, tangentMode=RCTM_User, time=1, value=1}
         ]}}}
      Scale Curve=1.5
     )
     Y=&FloatFromCurve(FloatCurve=DataInterface() {propertyValues {Curve {keys [<at 0 0 RCIM_Cubic>, {arriveTangent=6, interpMode=RCIM_Cubic, leaveTangent=6, time=0.1, value=5}, <at 1 6 RCIM_Cubic>]}}})
    )
    Scale Sprite Size Mode=::NewEnumerator2
   ) {module=ScaleSpriteSize001}
   &DynamicMaterialParameters(
    Index 0 Param 1=&FloatFromCurve(
     FloatCurve=DataInterface() {propertyValues {Curve {keys [
         <at 0 2 RCIM_Cubic>
         {arriveTangent=-2.5, leaveTangent=-2.5, time=0.2, value=0.1}
         {arriveTangent=1.125, leaveTangent=1.125, time=0.8, value=0}
         <at 1 1 RCIM_Cubic>
        ]}}}
     Scale Curve=0
    )
    Module.Index0Param0Write=true
    Module.Index0Param1Write=true
    Module.Index0Param2Write=true
    Module.Index0Param3Write=true
   )
  ]
  rendererClasses [&NiagaraSpriteRendererProperties]
  renderers [{rendererClass=&NiagaraSpriteRendererProperties}]
 }
 SegmentEmbers {
  simTarget=GPUComputeSim
  emitterUpdateScript [
   &EmitterState(Inactive Response=::NewEnumerator2, Life Cycle Mode=::NewEnumerator1, Loop Behavior=::NewEnumerator1, Loop Duration=0.25, UseLoopDelay=false)
   &SpawnPerUnit(Module.Use Max Movement Threshold=false, Module.Use Movement Tolerance=false, Spawn Spacing=100, Use Spawn Probability=false)
   &SpawnRate(SpawnRate=0, Use Spawn Probability=false) {enabled=false, module=SpawnRate001}
   &SpawnRate(SpawnRate=&Multiply_FloatByInt(Float=100, Integer=$User.NumSegments), Use Spawn Probability=false)
  ]
  particleSpawnScript [
   &InitializeParticle(
    Color=<rgba 1 0.4512 0 1>
    Lifetime=&RandomRangeFloat(Minimum=0.25, Maximum=0.7)
    Sprite Size Mode=::NewEnumerator3
    Uniform Sprite Size=&RandomRangeFloat(Minimum=3, Maximum=8)
    UsePositionOffset=false
    Write Material Random=true
   )
   &PartitionParticles(Number Of Partitions=$User.NumSegments)
   /SetVariables(Particles.OriginPos:NiagaraPosition=&TransformPosition(
     Position=&Multiply_VectorByFloat(
      Vector=&Multiply_Vector(
       A=&SelectVectorFromArray(Direct Array Index=$Particles.PartitionParticles.Partition, Vector Selection Array=$User.SegmentPositions)
       B=$User.RootScale
      )
      Float=&MakeFloatFromVector(Vector=$Engine.Owner.Scale, Channel=::NewEnumerator0)
     )
     Source Space=::Local
     Destination Space=::World
    )) {isSetParameters=true}
   &ShapeLocation(
    Additional Quaternion Rotation=false
    Box Size=<5 100 5>
    MaskSpawnGroup=false
    Offset Coordinate Space=::World
    Offset Mode=::NewEnumerator0
    Override Seed=false
    Rotation Mode=::NewEnumerator3
    Rotation Quaternion=&self:ScratchDynamicInput_02(Index:NiagaraInt32=$Particles.PartitionParticles.Partition, QuatArray:NiagaraDataInterfaceArrayQuat=$User.SegmentRotations)
    Shape Origin=$Particles.OriginPos
    Shape Primitive=::NewEnumerator2
    Surface Only=false
    UseOwnerScale=true
   )
   &ShapeLocation(MaskSpawnGroup=false, Override Seed=false, Sphere Radius=20, UseOwnerScale=true) {module=ShapeLocation001}
   &AddVelocity(
    Velocity=$Particles.ShapeLocation001.ShapeNormal
    Velocity Speed Scale=&RandomRangeFloat(Minimum=50, Maximum=500)
   ) {module=AddVelocity001}
   &KillParticles(Kill Particles=&InvertBool(Boolean:NiagaraBool=&self:ScratchDynamicInput(SegmentIndex:NiagaraInt32=$Particles.PartitionParticles.Partition, VisibilityArray:NiagaraDataInterfaceArrayBool=$User.ActiveArray)))
   &InheritVelocity(
    Inherited Velocity Amount Scale=&VectorFromFloat(Value=&RandomRangeFloat(Minimum=0.5, Maximum=2))
    Speed Limit=false
    UseSpeedThreshold=false
   )
   &AddVelocity(
    Cone Angle=79.6511
    Cone Axis=<-1 0 0>
    Module.UseConeFalloff=false
    Module.UseVelDistribution=false
    Override Seed=false
    Velocity Mode=::NewEnumerator2
    Velocity Speed=&RandomRangeFloat(Minimum=600, Maximum=1250)
   ) {enabled=false}
  ]
  particleUpdateScript [
   &ParticleState()
   &ScaleSpriteSize(
    Uniform Curve Index=&FN_LOD_Distance_Remap(Start Distance=1000, End Distance=5000, Minimum Output=0, Maximum Output=1, Clamp Output=true)
    Uniform Curve Scale=0.5
    Uniform Curve Sprite Scale=DataInterface() {propertyValues {Curve {keys [{interpMode=RCIM_Cubic, tangentMode=RCTM_User, time=0, value=1}, {interpMode=RCIM_Cubic, tangentMode=RCTM_User, time=1, value=2}]}}}
   )
   &Drag(Drag=4, Use Linear Drag=true, Use Rotational Drag=true)
   &CurlNoiseForce(
    Noise Frequency=10
    Noise Quality / Cost=::NewEnumerator0
    Noise Strength=&FloatFromCurve(
     FloatCurve=DataInterface() {propertyValues {Curve {keys [{tangentMode=RCTM_User, time=0, value=0}, {tangentMode=RCTM_User, time=1, value=1}]}}}
     Scale Curve=2500
    )
    Pan Noise=false
   )
   &AccelerationForce(Acceleration=<-2000 0 0>, Coordinate Space=::Local)
   &SolveForcesAndVelocity(Clamp Velocity=false, Limit Acceleration=false)
   &ScaleSpriteSizeBySpeed(Max Scale Factor=<2 5>)
   &ScaleColor(
    Scale RGB=&VectorFromFloat(Value=&FloatFromCurve(
      FloatCurve=DataInterface() {propertyValues {Curve {keys [
          {arriveTangent=9.0725, interpMode=RCIM_Cubic, leaveTangent=9.0725, tangentMode=RCTM_User, time=0, value=0}
          {arriveTangent=-2.5249, interpMode=RCIM_Cubic, leaveTangent=-2.5249, tangentMode=RCTM_User, time=0.1408, value=0.9847}
          {arriveTangent=-0.0405, interpMode=RCIM_Cubic, leaveTangent=-0.0405, tangentMode=RCTM_User, time=1, value=0}
         ]}}}
      Scale Curve=5
     ))
    ScaleA=true
    ScaleRGB=true
   )
  ]
  rendererClasses [&NiagaraSpriteRendererProperties]
  renderers [{rendererClass=&NiagaraSpriteRendererProperties}]
 }
 SegmentProjectile {
  emitterUpdateScript [
   &EmitterState()
   &SpawnBurst_Instantaneous(Spawn Count=$User.NumSegments, Use Spawn Probability=false, UseLoopCountLimit=false)
   /SetVariables(Emitter.TrailDirection:Vector3f=&TransformVector(Vector=<-1 0 0>, Destination Space=::World)) {isSetParameters=true}
   /SetVariables(Emitter.TrailTargetWS:Vector3f=&TransformPosition(Position=$System.TrailTargetPosition, Source Space=::Local, Destination Space=::World)) {isSetParameters=true}
   /SetVariables() {isSetParameters=true}
  ]
  particleSpawnScript [
   &InitializeParticle(
    Color=<rgba 30 9.9276 0 1>
    Lifetime=5
    Mesh Scale=&MakeVector(
     X=1
     Y=&Multiply_Float(
      A=&Divide_Float(
       A=&Multiply_Float(A=&MakeFloatFromInt(Int=$User.SegmentWidth), B=2)
       B=100
      )
      B=&MakeFloatFromVector(Vector=$User.RootScale, Channel=::NewEnumerator0)
     )
     Z=0.5
    )
    Mesh Scale Mode=::NewEnumerator4
    Sprite Size Mode=::NewEnumerator3
    Uniform Sprite Size=210.8758
    UsePositionOffset=false
    Write Material Random=true
   )
   &PartitionParticles(Number Of Partitions=$User.NumSegments)
   /SetVariables(Particles.DataIndex:NiagaraInt32=/Niagara/DynamicInputs/Subtract/Subtract_Integer(A:NiagaraInt32=$User.NumSegments, B:NiagaraInt32=$Particles.PartitionParticles.Partition)) {isSetParameters=true, module=SetVariables_#16}
   /SetVariables(Particles.Position=&Multiply_Vector(
     A=&SelectVectorFromArray(Direct Array Index=$Particles.PartitionParticles.Partition, Vector Selection Array=$User.SegmentPositions)
     B=$User.RootScale
    )) {isSetParameters=true, module=SetVariables_#17}
   /SetVariables(Particles.Position=&AddVectorToPosition(Position=$Particles.Position, Vector=<0 0 0>)) {isSetParameters=true, module=SetVariables_#18}
   &InitialMeshOrientation(
    Mesh Orientation Mode=::NewEnumerator6
    Orientation Quaternion:Quat4f=&self:ScratchDynamicInput_02(Index:NiagaraInt32=$Particles.PartitionParticles.Partition, QuatArray:NiagaraDataInterfaceArrayQuat=$User.SegmentRotations)
    Use Rotation Vector=false
   )
   /SetVariables() {isSetParameters=true, module=SetVariables_#19}
  ]
  particleUpdateScript [
   &ParticleState()
   &ScaleSpriteSize(
    Scale Sprite Size Mode=::NewEnumerator0
    Uniform Scale Factor=&Multiply_Float(
     A=&Multiply_Float(A=&MakeFloatFromVector2d(Vector 2D=$Particles.Initial.SpriteSize), B=0.005)
     B=&MakeFloatFromVector(Vector=$Engine.Owner.Scale, Channel=::NewEnumerator0)
    )
   )
   &ScaleMeshSize(Scale Factor=&MakeVector(
     X=1
     Y=1
     Z=&Multiply_Float(A=1, B=&MakeFloatFromVector(Vector=$Engine.Owner.Scale, Channel=::NewEnumerator0))
    ))
   /SetVariables(Particles.bIsHit:NiagaraBool=&InvertBool(Boolean:NiagaraBool=&self:ScratchDynamicInput(SegmentIndex:NiagaraInt32=$Particles.PartitionParticles.Partition, VisibilityArray:NiagaraDataInterfaceArrayBool=$User.ActiveArray))) {isSetParameters=true, module=SetVariables_#20}
   &Color(Color=&MakeCustomLinearColorFromBool(Bool=$Particles.bIsHit, True Color=<rgba 1 0 0 1>, False Color=<rgba 1 0.7128 0.0484 1>)) {enabled=false}
   /SetVariables(
    Particles.isLeftNeighbourViz:NiagaraBool=&self:ScratchDynamicInput_01(
     BoolArray:NiagaraDataInterfaceArrayBool=$User.ActiveArray
     NeighbourIndex:NiagaraInt32=&Subtract_Integer(A:NiagaraInt32=$Particles.PartitionParticles.Partition, B:NiagaraInt32=1)
    )
    Particles.isRightNeighbourViz:NiagaraBool=&self:ScratchDynamicInput_01(
     BoolArray:NiagaraDataInterfaceArrayBool=$User.ActiveArray
     NeighbourIndex:NiagaraInt32=&Subtract_Integer(A:NiagaraInt32=$Particles.PartitionParticles.Partition, B:NiagaraInt32=-1)
    )
   ) {enabled=false, isSetParameters=true, module=SetVariables_#21}
   /SetVariables(
    Particles.isLeftNeighbourViz:NiagaraBool=&Boolean_AndOperation(
     Bool1=&SetBoolByIntComparison(A=$Particles.PartitionParticles.Partition, B=0, Comparison Type=::NewEnumerator0)
     Bool2=$Particles.isLeftNeighbourViz
    )
    Particles.isRightNeighbourViz:NiagaraBool=&Boolean_AndOperation(
     Bool1=&SetBoolByIntComparison(
      A=$Particles.PartitionParticles.Partition
      B=&Subtract_Integer(A:NiagaraInt32=$User.NumSegments, B:NiagaraInt32=2)
      Comparison Type=::NewEnumerator7
     )
     Bool2=$Particles.isRightNeighbourViz
    )
   ) {enabled=false, isSetParameters=true, module=SetVariables_#22}
   /SetVariables(Particles.PrjThickness:NiagaraFloat=&Multiply_Float(
     A=&Multiply_Float(A=64, B=$User.BaseThicknessMultiplier)
     B=&MakeFloatFromVector(Vector=$Engine.Owner.Scale, Channel=::NewEnumerator0)
    )) {isSetParameters=true}
   &DynamicMaterialParameters(
    Index 0 Param 1=&MakeFloatFromVector(Vector=$Emitter.TrailTargetWS, Channel=::NewEnumerator0)
    Index 0 Param 2=&MakeFloatFromVector(Vector=$Emitter.TrailTargetWS, Channel=::NewEnumerator1)
    Index 0 Param 3=&MakeFloatFromVector(Vector=$Emitter.TrailTargetWS, Channel=::NewEnumerator2)
    Index 0 Param 4=200
    Index 1 Param 1=&MakeFloatFromVector(Vector=$Emitter.TrailDirection, Channel=::NewEnumerator0)
    Index 1 Param 2=&MakeFloatFromVector(Vector=$Emitter.TrailDirection, Channel=::NewEnumerator1)
    Index 1 Param 3=&MakeFloatFromVector(Vector=$Emitter.TrailDirection, Channel=::NewEnumerator2)
    Index 1 Param 4=&FloatFromCurve(
     FloatCurve=DataInterface() {propertyValues {Curve {keys [
         {arriveTangent=4.1785, interpMode=RCIM_Cubic, leaveTangent=4.1785, tangentMode=RCTM_User, time=0, value=0}
         {arriveTangent=0.0224, interpMode=RCIM_Cubic, leaveTangent=0.0224, tangentMode=RCTM_User, time=1, value=1}
        ]}}}
     CurveIndex=&Multiply_Float(A=$System.Age, B=2.25)
     Scale Curve=300
    )
    Index 2 Param 1=&MakeCustomFloatFromBool(Bool=$User.isMassive, True Float=3.5, False Float=2)
    Index 2 Param 2=&RandomRangeFloat(Minimum=0, Maximum=3)
    Index 2 Param 3=&Multiply_Float(A=$User.HeaderTiling, B=2)
    Index 2 Param 4=$Particles.PrjThickness
    Module.Index0Param0Write=true
    Module.Index0Param1Write=true
    Module.Index0Param2Write=true
    Module.Index0Param3Write=true
    Module.Index1Param0Write=true
    Module.Index1Param1Write=true
    Module.Index1Param2Write=true
    Module.Index1Param3Write=true
    Module.Index2Param0Write=true
    Module.Index2Param1Write=true
    Module.Index2Param2Write=true
    Module.Index2Param3Write=true
    Module.Index3Param0Write=true
    Module.Index3Param1Write=true
    Module.Index3Param2Write=true
    Module.Index3Param3Write=true
    Write Parameter Index 1=true
    Write Parameter Index 2=true
    Write Parameter Index 3=true
   )
   &KillParticles(Kill Particles=$Particles.bIsHit)
   /SetVariables(
    Particles.TrailColor:LinearColor=&MultiplyLinearColor(
     Linear Color=<rgba 1 0.298 0 1>
     Scale Factor=&MakeFloat4FromVectorAndFloat(
      "Vector (RGB)"=<1 1 1>
      "Float (Alpha)"=&Multiply_Float(
       A=&FloatFromCurve(
        FloatCurve=DataInterface() {propertyValues {Curve {keys [<at 0 1 RCIM_Cubic>, {interpMode=RCIM_Cubic, tangentMode=RCTM_User, time=1, value=0}]}}}
        CurveIndex=$System.GlobalFade
       )
       B=&FloatFromCurve(FloatCurve=DataInterface() {propertyValues {Curve {keys [
            {arriveTangent=9.0725, interpMode=RCIM_Cubic, time=0, value=0}
            {arriveTangent=0.8929, interpMode=RCIM_Cubic, leaveTangent=0.8929, time=0.1408, value=0.9847}
            {arriveTangent=-1.146, interpMode=RCIM_Cubic, leaveTangent=-1.146, time=0.7, value=0.625}
            {interpMode=RCIM_Cubic, leaveTangent=-0.0405, time=1, value=0}
           ]}}})
      )
     )
    )
    Particles.TrailScale:Vector3f=&MakeVector(X=0.1, Y=1, Z=$Particles.PrjThickness)
   ) {isSetParameters=true}
   /SetVariables(Particles.AirglowColor:LinearColor=&MultiplyLinearColor(Linear Color=<rgba 1 0.1363 0 0.05>, Scale Factor=&MakeVector4(X=1, Y=1, Z=1, W=$System.GlobalFade))) {isSetParameters=true}
   &FN_Light_Attributes(
    Color=$Particles.Color
    Create Color Attribute=true
    Create Position Attribute=true
    Create Radius Attribute=true
    Enable Particle Light Enable=true
    Light Enabled=true
    Position=&AddVectorToPosition(Position=$Particles.Position, Vector=<0 0 0>)
    Radius=&Multiply_Float(
     A=&Multiply_Float(A=80, B=2.25)
     B=&MakeFloatFromVector(Vector=$Engine.Owner.Scale, Channel=::NewEnumerator0)
    )
    Scale Color=&Multiply_Float(
     A=$System.GlobalFade
     B=&Multiply_Float(
      A=&Waveform(
       Waveform Count=::NewEnumerator1
       Global Phase=&Add_Float(A=$Emitter.Age, B=&RandomRangeFloat(Minimum=0, Maximum=1))
       "Waveform [1]"=::NewEnumerator12
       "[1] Frequency"=8
       "[1] Amplitude Min/Max"=<0.5 2>
       "[1] Bias"=0.2
       "[2] Average"=false
       "Waveform [2]"=::NewEnumerator12
       "[2] Input Position"=0
       "[2] Phase"=$Emitter.Age
       "[2] Phase Scale"=1
       "[2] Compound Frequency"=2
       "[2] Frequency"=12
       "[2] Amplitude Min/Max"=<0.25 1.5>
       "[2] Scale"=1
       "[2] Bias"=0.5
       "[2] Exponent"=1
       "[2] Pulse Duration"=0.1
       "[2] Clip Triangle Min/Max"=<0 0>
       "[2] Random Seed"=0
      )
      B=0.25
     )
    )
   )
   &ScaleColor(
    Scale Alpha=&FloatFromCurve(FloatCurve=DataInterface() {propertyValues {Curve {keys [<at 0 0>,<at 1 1 RCIM_Cubic>]}}}, CurveIndex=$System.GlobalFade)
    ScaleA=true
    ScaleRGB=true
   )
   &ScaleColor(
    Scale Alpha=&FloatFromCurve(
     FloatCurve=DataInterface() {propertyValues {Curve {keys [
         {interpMode=RCIM_Cubic, tangentMode=RCTM_User, time=0, value=0}
         {arriveTangent=2.9579, interpMode=RCIM_Cubic, leaveTangent=2.9579, tangentMode=RCTM_User, time=1, value=1}
        ]}}}
     CurveIndex=&FN_LOD_Distance_Remap(Start Distance=150, End Distance=2000, Minimum Output=0, Maximum Output=1, Clamp Output=true)
    )
    ScaleA=true
    ScaleRGB=false
   ) {module=ScaleColor001}
   /SetVariables(
    Particles.AirflowColor:LinearColor=&MultiplyLinearColor(Linear Color=<rgba 10 0.8587 0 1>, Scale Factor=&MakeVector4(X=1, Y=1, Z=1, W=$System.GlobalFade))
    Particles.AirflowParams:Vector4f=&MakeVector4(X=&MakeCustomFloatFromBool(Bool=$User.isMassive, True Float=2, False Float=1), Y=0, Z=0, W=0)
    Particles.AirflowSize:Vector3f=&Multiply_Vector(
     A=&MakeVector(X=1, Y=1, Z=&MakeFloatFromVector(Vector=$Engine.Owner.Scale, Channel=::NewEnumerator0))
     B=&MakeCustomVectorFromBool(Bool=$User.isMassive, True Vector=<1 1 0.35>, False Vector=<1 1 0.5>)
    )
   ) {isSetParameters=true}
  ]
  rendererClasses [
   &NiagaraSpriteRendererProperties
   &NiagaraMeshRendererProperties
   &NiagaraLightRendererProperties
  ]
  renderers [
   {rendererClass=&NiagaraSpriteRendererProperties}
   {rendererClass=&NiagaraMeshRendererProperties, rendererIndex=1}
   {rendererClass=&NiagaraMeshRendererProperties, rendererIndex=2}
   {rendererClass=&NiagaraLightRendererProperties, rendererIndex=3}
   {rendererClass=&NiagaraMeshRendererProperties, rendererIndex=4}
   {rendererClass=&NiagaraMeshRendererProperties, rendererIndex=5}
  ]
 }
 SegmentedTrails {
  emitterUpdateScript [
   &EmitterState()
   &SpawnBurst_Instantaneous(Spawn Count=&Multiply_Int(A=$User.NumSegments, B=1), Use Spawn Probability=false, UseLoopCountLimit=false)
   &SpawnBurst_Instantaneous(Spawn Count=$User.NumSegments, Spawn Group=1, Use Spawn Probability=false, UseLoopCountLimit=false) {enabled=false, module=SpawnBurst_Instantaneous001}
  ]
  particleSpawnScript [
   &InitializeParticle(
    Color=<rgba 20 0.7204 0 1>
    Lifetime=5
    Mesh Scale=&MakeVector(
     X=1
     Y=&Multiply_Float(
      A=&Divide_Float(
       A=&Multiply_Float(A=&MakeFloatFromInt(Int=$User.SegmentWidth), B=2)
       B=100
      )
      B=&MakeFloatFromVector(Vector=$User.RootScale, Channel=::NewEnumerator0)
     )
     Z=0.5
    )
    Mesh Scale Mode=::NewEnumerator4
    Sprite Size Mode=::NewEnumerator3
    Uniform Sprite Size=210.8758
    UsePositionOffset=false
    Write Material Random=true
   )
   &PartitionParticles(Number Of Partitions=$User.NumSegments)
   /SetVariables(Particles.DataIndex:NiagaraInt32=&Subtract_Integer(A:NiagaraInt32=$User.NumSegments, B:NiagaraInt32=$Particles.PartitionParticles.Partition)) {isSetParameters=true, module=SetVariables_#16}
   /SetVariables(Particles.Position=&Multiply_Vector(
     A=&SelectVectorFromArray(Direct Array Index=$Particles.PartitionParticles.Partition, Vector Selection Array=$User.SegmentPositions)
     B=$User.RootScale
    )) {isSetParameters=true, module=SetVariables_#17}
   /SetVariables(Particles.Position=&AddVectorToPosition(Position=$Particles.Position, Vector=<0 0 0>)) {isSetParameters=true, module=SetVariables_#18}
   &InitialMeshOrientation(
    Mesh Orientation Mode=::NewEnumerator6
    Orientation Quaternion:Quat4f=&self:ScratchDynamicInput_02(Index:NiagaraInt32=$Particles.PartitionParticles.Partition, QuatArray:NiagaraDataInterfaceArrayQuat=$User.SegmentRotations)
    Use Rotation Vector=false
   )
   /SetVariables(Particles.Position=&AddVectorToPosition(
     Position=$Particles.Position
     Vector=&RotateVectorByQuaternion(
      Vector To Rotate=&Multiply_VectorByFloat(
       Vector=&MaskVectorBySpawnGroup(Masked Vector=<0 -1 0>, Passthrough Vector=<0 1 0>, Spawn Group Mask=1)
       Float=0
      )
      Quaternion=$Particles.MeshOrientation
     )
    )) {isSetParameters=true}
   /SetVariables() {enabled=false, isSetParameters=true, module=SetVariables_#19}
   &InitialMeshOrientation(
    Mesh Orientation Mode=::NewEnumerator2
    Orientation Coordinate Space=::Local
    Orientation Vector=&NormalizeVector(
     Vector To Normalize=&Subtract_Vector(A=Linked() {linkedVariable {name=Particles.Position, type=NiagaraPosition}}, B=$System.TrailTargetPosition)
     Safe Normalize=true
     Safe Fallback Threshold=0
     Safe Fallback Vector=<0 0 0.001>
    )
    Use Rotation Vector=false
   ) {enabled=false, module=InitialMeshOrientation001}
   /SetVariables(Particles.MaterialRandom:NiagaraFloat=&RandomRangeFloat(Minimum=0, Maximum=1)) {isSetParameters=true}
  ]
  particleUpdateScript [
   &ParticleState()
   &ScaleMeshSize(Scale Factor=&MakeVector(X=1, Y=0.4, Z=1))
   /SetVariables(Particles.bIsHit:NiagaraBool=&InvertBool(Boolean:NiagaraBool=&self:ScratchDynamicInput(SegmentIndex:NiagaraInt32=$Particles.PartitionParticles.Partition, VisibilityArray:NiagaraDataInterfaceArrayBool=$User.ActiveArray))) {isSetParameters=true, module=SetVariables_#20}
   &Color(Color=&MakeCustomLinearColorFromBool(Bool=$Particles.bIsHit, True Color=<rgba 1 0 0 1>, False Color=<rgba 1 0.7128 0.0484 1>)) {enabled=false}
   /SetVariables(
    Particles.isLeftNeighbourViz:NiagaraBool=&self:ScratchDynamicInput_01(
     BoolArray:NiagaraDataInterfaceArrayBool=$User.ActiveArray
     NeighbourIndex:NiagaraInt32=&Subtract_Integer(A:NiagaraInt32=$Particles.PartitionParticles.Partition, B:NiagaraInt32=1)
    )
    Particles.isRightNeighbourViz:NiagaraBool=&self:ScratchDynamicInput_01(
     BoolArray:NiagaraDataInterfaceArrayBool=$User.ActiveArray
     NeighbourIndex:NiagaraInt32=&Subtract_Integer(A:NiagaraInt32=$Particles.PartitionParticles.Partition, B:NiagaraInt32=-1)
    )
   ) {enabled=false, isSetParameters=true, module=SetVariables_#21}
   /SetVariables(
    Particles.isLeftNeighbourViz:NiagaraBool=&Boolean_AndOperation(
     Bool1=&SetBoolByIntComparison(A=$Particles.PartitionParticles.Partition, B=0, Comparison Type=::NewEnumerator0)
     Bool2=$Particles.isLeftNeighbourViz
    )
    Particles.isRightNeighbourViz:NiagaraBool=&Boolean_AndOperation(
     Bool1=&SetBoolByIntComparison(
      A=$Particles.PartitionParticles.Partition
      B=&Subtract_Integer(A:NiagaraInt32=$User.NumSegments, B:NiagaraInt32=2)
      Comparison Type=::NewEnumerator7
     )
     Bool2=$Particles.isRightNeighbourViz
    )
   ) {enabled=false, isSetParameters=true, module=SetVariables_#22}
   &KillParticles(Kill Particles=$Particles.bIsHit)
   &ScaleColor(
    Scale Alpha=&FloatFromCurve(FloatCurve=DataInterface() {propertyValues {Curve {keys [<at 0 0>,<at 1 1 RCIM_Cubic>]}}}, CurveIndex=$System.GlobalFade)
    ScaleA=true
    ScaleRGB=true
   )
   &DynamicMaterialParameters(
    Index 0 Param 1=&Multiply_Float(A=45, B=&MakeFloatFromVector(Vector=$Engine.Owner.Scale, Channel=::NewEnumerator0))
    Module.Index0Param0Write=true
    Module.Index0Param1Write=true
    Module.Index0Param2Write=true
    Module.Index0Param3Write=true
   )
  ]
  rendererClasses [&NiagaraMeshRendererProperties]
  renderers [{rendererClass=&NiagaraMeshRendererProperties, rendererIndex=0}]
 }
 Trails {
  emitterUpdateScript [
   &EmitterState()
   &SpawnRate(SpawnRate=500, Use Spawn Probability=false)
  ]
  particleSpawnScript [
   &InitializeParticle(Color=<rgba 1 0.4231 0 2>, Lifetime=0.3, UsePositionOffset=false, Write Material Random=true)
   &AddVelocity(Velocity=<-2500 0 0>)
   &PartitionParticles(Number Of Partitions=$User.NumSegments, Write Partition to Ribbon ID=::NewEnumerator0)
   /SetVariables(Particles.Position=&TransformPosition(
     Position=&Multiply_Vector(
      A=&SelectVectorFromArray(Direct Array Index=$Particles.PartitionParticles.Partition, Vector Selection Array=$User.SegmentPositions)
      B=<1 1 1>
     )
     Source Space=::Local
     Destination Space=::World
     Apply Scale=true
    )) {isSetParameters=true}
   &KillParticles(Kill Particles=&InvertBool(Boolean:NiagaraBool=&self:ScratchDynamicInput(SegmentIndex:NiagaraInt32=$Particles.PartitionParticles.Partition, VisibilityArray:NiagaraDataInterfaceArrayBool=$User.ActiveArray))) {module=KillParticles001}
  ]
  particleUpdateScript [
   &ParticleState()
   /SetVariables(Particles.RibbonWidth=&FloatFromCurve(FloatCurve=DataInterface() {propertyValues {Curve {keys [<at 0 1.5>,<at 1 1>]}}}, Scale Curve=$User.TrailingRibbonWidth)) {isSetParameters=true}
   &AccelerationForce(Acceleration=<0 0 -200>)
   &SolveForcesAndVelocity(Clamp Velocity=false, Limit Acceleration=false)
   /SetVariables(Particles.RibbonFacing=$Engine.Owner.SystemZAxis) {isSetParameters=true}
   &ScaleColor(
    Scale Alpha=&Multiply_Float(
     A=$System.GlobalFade
     B=&FloatFromCurve(
      FloatCurve=DataInterface() {propertyValues {Curve {keys [
          {interpMode=RCIM_Cubic, tangentMode=RCTM_User, time=0, value=0}
          {arriveTangent=2.9579, interpMode=RCIM_Cubic, leaveTangent=2.9579, tangentMode=RCTM_User, time=1, value=1}
         ]}}}
      CurveIndex=&FN_LOD_Distance_Remap(Start Distance=200, End Distance=3000, Minimum Output=0, Maximum Output=1, Clamp Output=true)
     )
    )
    ScaleA=true
    ScaleRGB=true
   )
   &DynamicMaterialParameters(
    Index 0 Param 1=0.1
    Index 2 Param 1=&RandomRangeFloat(Minimum=0, Maximum=1)
    Index 2 Param 2=&Multiply_FloatByInt(Float=0.355, Integer=$Particles.PartitionParticles.Partition)
    Index 2 Param 3=&FloatFromCurve(FloatCurve=DataInterface() {propertyValues {Curve {keys [<at 0 1>,<at 1 0>]}}}, CurveIndex=$System.NormalizedLoopAge, Scale Curve=4)
    Module.Index0Param0Write=true
    Module.Index0Param1Write=true
    Module.Index0Param2Write=true
    Module.Index0Param3Write=true
    Module.Index2Param0Write=true
    Module.Index2Param1Write=true
    Module.Index2Param2Write=true
    Module.Index2Param3Write=true
    Write Parameter Index 2=true
   )
   &KillParticles(Kill Particles=&InvertBool(Boolean:NiagaraBool=&self:ScratchDynamicInput(SegmentIndex:NiagaraInt32=$Particles.PartitionParticles.Partition, VisibilityArray:NiagaraDataInterfaceArrayBool=$User.ActiveArray)))
  ]
  rendererClasses [&NiagaraSpriteRendererProperties, &NiagaraRibbonRendererProperties]
  renderers [
   {rendererClass=&NiagaraSpriteRendererProperties}
   {rendererClass=&NiagaraRibbonRendererProperties, rendererIndex=1}
  ]
 }
}
summary {
 systemName=NS_AnnoyedMop_Prj_Proto
 emitters [
  {
   emitterName=SegmentProjectile
   rendererClasses [
    &NiagaraSpriteRendererProperties
    &NiagaraMeshRendererProperties
    &NiagaraLightRendererProperties
   ]
  }
  {emitterName=SegmentEmbers, simTarget=GPUComputeSim, rendererClasses [&NiagaraSpriteRendererProperties]}
  {bEnabled=false, emitterName=FlameLicks, rendererClasses [&NiagaraSpriteRendererProperties]}
  {bEnabled=false, emitterName=LongFlamelicks, rendererClasses [&NiagaraSpriteRendererProperties]}
  {emitterName=SegmentedTrails, rendererClasses [&NiagaraMeshRendererProperties]}
  {emitterName=Trails, rendererClasses [&NiagaraSpriteRendererProperties, &NiagaraRibbonRendererProperties]}
  {emitterName=GPU_Collided_Smoke, simTarget=GPUComputeSim, rendererClasses [&NiagaraSpriteRendererProperties]}
 ]
 userVariables [
  User.ActiveArray {
   type=NiagaraDataInterfaceArrayBool
   description=""
   defaultValue {
    struct=NiagaraExt_VariableValue_DataInterface
    value {
     dataInterface=&self:NiagaraDataInterfaceArrayBool_0
     dataInterfaceClass=/Script/Niagara.NiagaraDataInterfaceArrayBool
    }
   }
  }
  User.BaseThicknessMultiplier {type=NiagaraFloat, defaultValue=1, description=""}
  User.DebugShapes {type=NiagaraBool}
  User.DustLocation {type=NiagaraPosition, description="", defaultValue {}}
  User.GlobalScale {type=NiagaraFloat, defaultValue=0, description=""}
  User.GroundDust {type=NiagaraBool, defaultValue=true}
  User.HeaderTiling {type=NiagaraFloat, defaultValue=1.5, description=""}
  User.isMassive {type=NiagaraBool}
  User.NumSegments {type=NiagaraInt32, defaultValue=5}
  User.PrjSize {type=NiagaraInt32}
  User.RootScale {type=Vector3f, description="", defaultValue {struct=Vector3f, value=<1 1 1>}}
  User.ScaleMovement {type=NiagaraFloat, defaultValue=0, description=""}
  User.SegmentMesh {
   type=StaticMesh
   description=""
   defaultValue {struct=NiagaraExt_VariableValue_Object, value {object=null, objectClass=&StaticMesh}}
  }
  User.SegmentPositions {
   type=NiagaraDataInterfaceArrayFloat3
   description=""
   defaultValue {
    struct=NiagaraExt_VariableValue_DataInterface
    value {
     dataInterface=&self:NiagaraDataInterfaceArrayFloat3_0
     dataInterfaceClass=&NiagaraDataInterfaceArrayFloat3
    }
   }
  }
  User.SegmentRotations {
   type=NiagaraDataInterfaceArrayQuat
   description=""
   defaultValue {
    struct=NiagaraExt_VariableValue_DataInterface
    value {
     dataInterface=&self:NiagaraDataInterfaceArrayQuat_0
     dataInterfaceClass=/Script/Niagara.NiagaraDataInterfaceArrayQuat
    }
   }
  }
  User.SegmentWidth {type=NiagaraInt32, defaultValue=60}
  User.TargetSM {
   type=NiagaraDataInterfaceSkeletalMesh
   defaultValue {
    struct=NiagaraExt_VariableValue_DataInterface
    value {dataInterface=&self:NiagaraDataInterfaceSkeletalMesh_0}
   }
  }
  User.TrailingRibbonWidth {type=NiagaraFloat, defaultValue=250, description=""}
 ]
}
